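# Body is a placeholder-style material in the Bake_Meshes folder of the F_MED_FlossYawn medium female player body cosmetic.
# It uses the Default Lit shading model and holds a single constant cyan color node, with no textures, parameters or visible connections.
# Its asset import data indicates it comes from an Interchange import.
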Material /BRCosmetics/Characters/Player/Female/Medium/Bodies/F_MED_FlossYawn/Meshes/Bake_Meshes/Body
graph {
 properties {assetImportData="<self>:InterchangeAssetImportData_0", usedShadingModels=Default Lit}
 nodes [Constant3Vector() {constant=<rgba 0 0.8 0.8 1>}]
}
